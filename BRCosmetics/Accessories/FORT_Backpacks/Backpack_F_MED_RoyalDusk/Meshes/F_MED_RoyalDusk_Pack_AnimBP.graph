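# Animation Blueprint for the RoyalDusk backpack cosmetic that gives its star bone a floating, secondary motion.
# The input pose feeds cached AnimDynamics and planar-limited trail controller branches, selected by bAnimDynamicsEnabled, with an extra star lift during backward jumps.
# An additive bob from layered sine waves applies on top except while jumping or falling.
# The EventGraph tunes dynamics by jump, fall and crouch state, and trail alpha by 2D movement.
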AnimBlueprint /BRCosmetics/Accessories/FORT_Backpacks/Backpack_F_MED_RoyalDusk/Meshes/F_MED_RoyalDusk_Pack_AnimBP
target_skeleton=/Game/Accessories/FORT_Backpacks/Backpack_F_MED_RoyalDusk/Meshes/F_MED_RoyalDusk_Pack_Skeleton
graphs [
 AnimGraph {nodes [
   n1 = Use cached pose 'AnimDynamics'()
   n2 = Use cached pose 'TrailController'()
   n3 = Get bAnimDynamicsEnabled() -> (bAnimDynamicsEnabled:bool)
   n4 = "Transform (Modify) Bone - Bone: star"!(
    ComponentPose=Local To Component(LocalPose=Blend Poses by bool(BlendPose_0=@n1, BlendPose_1=@n2, bActiveValue=@n3) {blendType=Linear})
    bAlphaBoolEnabled:bool=Get bIsNotJumpingOrFalling() -> (bIsNotJumpingOrFalling:bool)
    Translation=Get BackpackFloating() -> (BackpackFloating:Vector)
   ) -> (Pose) {
    alphaInputType=Bool
    translationMode=BMM_Additive
    translationSpace=BCS_BoneSpace
    alphaBoolBlend {blendInTime=0.2, blendOption=Linear, blendOutTime=1}
    boneToModify {boneName=star}
   }
   Output Pose(Result=Component To Local(ComponentPose=@n4)) {layerGroup=None}
   Save cached pose 'BasePose'(Pose=Input Pose()) {cacheName=BasePose}
   n5 = Trail controller - Bone: star(ComponentPose=Local To Component(LocalPose=Use cached pose 'BasePose'()), Alpha=Get TrailAlpha() -> (TrailAlpha:double)) {
    bInvertChainBoneAxis=true
    bLimitRotation=true
    bUsePlanarLimit=true
    chainBoneAxis=Y
    relaxationSpeedScale=0.95
    stretchLimit=5
    baseJoint {boneName=pack}
    planarLimits [
     {drivingBone {boneName=root}, planeTransform {location=<0 -1 0>, rotation=<rot 0 0 -90>, scale=<1 1 1>}}
     {drivingBone {boneName=root}, planeTransform {location=<0 -8 0>, rotation=<rot 0 0 90>, scale=<1 1 1>}}
     {drivingBone {boneName=pack}, planeTransform {location=<5 0 0>, rotation=<rot 90 0 0>, scale=<1 1 1>}}
    ]
    rotationLimits [{limitMax=<0 0 0>, limitMin=<0 0 0>}, {limitMax=<0 0 0>, limitMin=<0 0 0>}]
    rotationOffsets [<0 0 0>, <0 0 0>]
    trailBone {boneName=star}
    trailRelaxationSpeed {editorCurveData {keys [<at 0 30 RCIM_Linear>,<at 1 30 RCIM_Linear>]}}
   }
   Save cached pose 'TrailController'(Pose=Component To Local(ComponentPose=@n5)) {cacheName=TrailController}
   n6 = Anim Dynamics - Bone: star!(
    ComponentPose=Local To Component(LocalPose=Use cached pose 'BasePose'())
    Alpha=Get AnimDynamicsAlpha() -> (AnimDynamicsAlpha:float)
    ComponentLinearAccScale:Vector=Get LinearAccScale() -> (LinearAccScale:Vector)
    ComponentLinearVelScale:Vector=Get LinearVelScale() -> (LinearVelScale:Vector)
   ) -> (Pose) {
    alpha=0
    angularBiasOverride=1
    angularDampingOverride=0.75
    angularSpringConstant=20
    bOverrideLinearDamping=true
    bUseGravityOverride=true
    bUsePlanarLimit=false
    gravityScale=0
    linearDampingOverride=0.9
    simulationSpace=BoneRelative
    alphaBoolBlend {blendOption=Linear}
    boundBone {boneName=star}
    componentAppliedLinearAccClamp=<100 100 100>
    componentLinearAccScale=<0.7 0.7 1.5>
    componentLinearVelScale=<0.7 0.7 1.5>
    physicsBodyDefinitions [{
      sphereCollisionRadius=0
      boundBone {boneName=star}
      boxExtents=<20 8 30>
      constraintSetup {angularLimitsMax=<0 0 0>, angularLimitsMin=<0 0 0>, angularTarget=<0 0 0>, linearAxesMax=<0.5 0 3.5>, linearAxesMin=<-0.5 0 -1>}
      localJointOffset=<0 0 0>
     }]
    relativeSpaceBone {boneName=root}
   }
   n7 = "Transform (Modify) Bone - Bone: star"(ComponentPose=@n6, bAlphaBoolEnabled=Get bIsBackwardJumping() -> (bIsBackwardJumping:bool), Translation="0,0,2") {
    alphaInputType=Bool
    translationMode=BMM_Additive
    translationSpace=BCS_BoneSpace
    alphaBoolBlend {blendInTime=0.4, blendOption=Linear, blendOutTime=1}
    boneToModify {boneName=star}
    translation {z=2}
   }
   Save cached pose 'AnimDynamics'(Pose=Component To Local(ComponentPose=@n7)) {cacheName=AnimDynamics}
  ]}
 EventGraph {nodes [
   Event BlueprintUpdateAnimation()
   > n1 = Sequence!(execute) -> (then_0, then_1, then_2:Exec)
   n2 = float * float(A=GetGameTimeInSeconds(), B=3) {is=title K2Node_PromotableOperator}
   n3 = float * float(A="Sin (Radians)"(A=float * float(A=1.21, B=@n2)), B=0.1) {is=title K2Node_CommutativeAssociativeBinaryOperator}
   n4 = float * float(A="Sin (Radians)"(A=float * float(A=0.56, B=@n2)), B=0.8)
   n5 = float * float(A="Sin (Radians)"(A=float * float(A=0.8, B=@n2)), B=1.2)
   > Set BackpackFloating(BackpackFloating:Vector=vector * vector(A=MakeVector(X=@n4, Y=@n3, Z=@n5), B:float=1.2) -> (ReturnValue=~)) -> (Output_Get:Vector)
   n6 = Branch(execute=@n1.then_1, Condition=OR Boolean(A=Get bIsJumping(), B=Get bIsFalling()))
   > Set bIsNotJumpingOrFalling(bIsNotJumpingOrFalling:bool=false) -> (Output_Get:bool)
   > Set LinearAccScale(LinearAccScale:Vector="0,0,0.15") -> (Output_Get:Vector)
   > Set LinearVelScale(LinearVelScale:Vector="0,0,0.2") -> (Output_Get:Vector)
   > Set AnimDynamicsAlpha(AnimDynamicsAlpha:float=0.8) -> (Output_Get:float)
   n7 = Get PawnMovementDirectionAzimuth() -> (PawnMovementDirectionAzimuth:float)
   > Set bIsBackwardJumping(bIsBackwardJumping:bool=OR Boolean(A="float < float"(A=@n7, B=-150), B="float > float"(A=@n7, B=150))) -> (Output_Get:bool)
   Set bIsNotJumpingOrFalling(execute=@n6.else, bIsNotJumpingOrFalling=true)
   > Set AnimDynamicsAlpha(AnimDynamicsAlpha=1)
   > Set bIsBackwardJumping(bIsBackwardJumping=false)
   > n8 = Branch(Condition=Get bIsCrouching())
   > Set LinearAccScale(LinearAccScale="0.5,0.5,0.1")
   > Set LinearVelScale(LinearVelScale="0.7,0.7,0.5")
   Set LinearAccScale(execute=@n8.else, LinearAccScale="0.6,0.6,1.5")
   > Set LinearVelScale(LinearVelScale="0.7,0.7,1.5")
   Set TrailAlpha(execute=@n1.then_2, TrailAlpha:double=SelectFloat(A=0.4, B=0, bPickA=Get bIsMoving2D())) -> (Output_Get:double)
   TryGetPawnOwner()
  ]}
]
variables [
 AnimDynamicsAlpha {type=number, default=1}
 BackMovingAlpha {type=number, default=0.9}
 BackpackFloating {type=Vector}
 bIsBackwardJumping {type=boolean}
 bIsNotJumpingOrFalling {type=boolean}
 Gravity Override {type=Vector}
 LinearAccScale {type=Vector, default=<0.7 0.7 1.5>}
 LinearVelScale {type=Vector, default=<0.7 0.7 1.5>}
 TrailAlpha {type=number}
]
